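# HUD widget Blueprint based on MultiInteractDisplayWidgetBase that displays multi-interact icons and an interaction count.
# An inherited visibility switcher holds two layouts: a scaled-down row of four identical icons, and a single scaled icon beside an "x 99" placeholder count.
# The event graph is empty, so the Blueprint only supplies layout and styling around the parent's inherited switcher, icon container and count text.
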WidgetBlueprint /Game/UI/InGame/HUD/MultiInteract/MultiInteractDisplayWidget
graphs [EventGraph {links [], nodes []}]
variables []
widgets {
 info {
  inheritedWidgetCount=3
  parentClass=/Script/FortniteUI.MultiInteractDisplayWidgetBase
  rootWidgetClass=&CommonVisibilitySwitcher
  widgetCount=11
 }
 widgets [
  VisibilitySwitcher {widgetClassPath=&CommonVisibilitySwitcher, bInherited=true, widget=VisibilitySwitcher}
  ScaleBox_0 {
   widgetClassPath=&ScaleBox
   parent=VisibilitySwitcher
   widget=ScaleBox_0
   properties {stretch=UserSpecified, userSpecifiedScale=0.4}
   slot {class=CommonVisibilitySwitcherSlot}
  }
  HorBoxInteractionIconContainer {
   widgetClassPath=&HorizontalBox
   bInherited=true
   parent=ScaleBox_0
   widget=HorBoxInteractionIconContainer
   slot {class=ScaleBoxSlot}
  }
  InteractIcon {
   widgetClassPath=&Image
   parent=HorBoxInteractionIconContainer
   widget=InteractIcon
   properties {brush {
     bIsDynamicallyLoaded=false
     drawAs=Image
     imageType=NoImage
     mirroring=NoMirror
     resourceName=None
     resourceObject=&/T_MultiInteractIcon
     tiling=NoTile
     imageSize=<64 64>
     margin=<ltrb 0 0 0 0>
     outlineSettings {
      bUseBrushTransparency=false
      roundingType=HalfHeightRadius
      width=0
      color {colorUseRule=UseColor_Specified, specifiedColor=<rgba 0 0 0 0>}
      cornerRadii=<0 0 0 0>
     }
     tintColor {colorUseRule=UseColor_Specified, specifiedColor=<rgba 1 1 1 1>}
     uVRegion {bIsValid=false, max=<0 0>, min=<0 0>}
    }}
   slot {class=HorizontalBoxSlot, properties {verticalAlignment=VAlign_Center, padding=<ltrb 0 0 5 0>}}
  }
  InteractIcon2 {
   widgetClassPath=&Image
   parent=HorBoxInteractionIconContainer
   widget=InteractIcon2
   properties {brush {
     bIsDynamicallyLoaded=false
     drawAs=Image
     imageType=NoImage
     mirroring=NoMirror
     resourceName=None
     resourceObject=&T_MultiInteractIcon
     tiling=NoTile
     imageSize=<64 64>
     margin=<ltrb 0 0 0 0>
     outlineSettings {
      bUseBrushTransparency=false
      roundingType=HalfHeightRadius
      width=0
      color {colorUseRule=UseColor_Specified, specifiedColor=<rgba 0 0 0 0>}
      cornerRadii=<0 0 0 0>
     }
     tintColor {colorUseRule=UseColor_Specified, specifiedColor=<rgba 1 1 1 1>}
     uVRegion {bIsValid=false, max=<0 0>, min=<0 0>}
    }}
   slot {class=HorizontalBoxSlot, properties {padding=<ltrb 0 0 5 0>}}
  }
  InteractIcon3 {
   widgetClassPath=&Image
   parent=HorBoxInteractionIconContainer
   widget=InteractIcon3
   properties {brush {
     bIsDynamicallyLoaded=false
     drawAs=Image
     imageType=NoImage
     mirroring=NoMirror
     resourceName=None
     resourceObject=&T_MultiInteractIcon
     tiling=NoTile
     imageSize=<64 64>
     margin=<ltrb 0 0 0 0>
     outlineSettings {
      bUseBrushTransparency=false
      roundingType=HalfHeightRadius
      width=0
      color {colorUseRule=UseColor_Specified, specifiedColor=<rgba 0 0 0 0>}
      cornerRadii=<0 0 0 0>
     }
     tintColor {colorUseRule=UseColor_Specified, specifiedColor=<rgba 1 1 1 1>}
     uVRegion {bIsValid=false, max=<0 0>, min=<0 0>}
    }}
   slot {class=HorizontalBoxSlot, properties {padding=<ltrb 0 0 5 0>}}
  }
  InteractIcon4 {
   widgetClassPath=&Image
   parent=HorBoxInteractionIconContainer
   widget=InteractIcon4
   properties {brush {
     bIsDynamicallyLoaded=false
     drawAs=Image
     imageType=NoImage
     mirroring=NoMirror
     resourceName=None
     resourceObject=&T_MultiInteractIcon
     tiling=NoTile
     imageSize=<64 64>
     margin=<ltrb 0 0 0 0>
     outlineSettings {
      bUseBrushTransparency=false
      roundingType=HalfHeightRadius
      width=0
      color {colorUseRule=UseColor_Specified, specifiedColor=<rgba 0 0 0 0>}
      cornerRadii=<0 0 0 0>
     }
     tintColor {colorUseRule=UseColor_Specified, specifiedColor=<rgba 1 1 1 1>}
     uVRegion {bIsValid=false, max=<0 0>, min=<0 0>}
    }}
   slot {class=HorizontalBoxSlot}
  }
  HorizontalBox_70 {widgetClassPath=&HorizontalBox, parent=VisibilitySwitcher, widget=HorizontalBox_70, slot {class=CommonVisibilitySwitcherSlot}}
  ScaleBox_1 {
   widgetClassPath=&ScaleBox
   parent=HorizontalBox_70
   widget=ScaleBox_1
   properties {stretch=UserSpecified, userSpecifiedScale=0.4}
   slot {class=HorizontalBoxSlot, properties {verticalAlignment=VAlign_Center, padding=<ltrb 0 0 5 0>}}
  }
  InteractIcon_1 {
   widgetClassPath=&Image
   parent=ScaleBox_1
   widget=InteractIcon_1
   properties {brush {
     bIsDynamicallyLoaded=false
     drawAs=Image
     imageType=NoImage
     mirroring=NoMirror
     resourceName=None
     resourceObject=&T_MultiInteractIcon
     tiling=NoTile
     imageSize=<64 64>
     margin=<ltrb 0 0 0 0>
     outlineSettings {
      bUseBrushTransparency=false
      roundingType=HalfHeightRadius
      width=0
      color {colorUseRule=UseColor_Specified, specifiedColor=<rgba 0 0 0 0>}
      cornerRadii=<0 0 0 0>
     }
     tintColor {colorUseRule=UseColor_Specified, specifiedColor=<rgba 1 1 1 1>}
     uVRegion {bIsValid=false, max=<0 0>, min=<0 0>}
    }}
   slot {class=ScaleBoxSlot}
  }
  Text_InteractionCount {
   widgetClassPath=&CommonTextBlock
   bInherited=true
   parent=HorizontalBox_70
   widget=Text_InteractionCount
   properties {
    lineHeightPercentage=0.9
    text=x 99
    wrapTextAt=320
    font {
     bForceMonospaced=false
     bMaterialIsStencil=false
     fontMaterial=null
     fontObject=&BurbankSmall
     letterSpacing=0
     monospacedWidth=1
     size=15
     skewAmount=0
     typefaceFontName=Black
     outlineSettings {
      bApplyOutlineToDropShadows=false
      bMiteredCorners=false
      bSeparateFillAlpha=true
      outlineMaterial=&M-FontShadow
      outlineSize=2
      outlineColor=<rgba 0 0 0 0.5>
     }
    }
    margin=<ltrb 0 -2 0 0>
   }
   slot {class=HorizontalBoxSlot}
  }
 ]
}
